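# Niagara system for the bow-and-arrow effect of the SeleneCobra player emote, timed by user shoot and explode times.
# ArrowTip (sprite, light and component renderers), spinning LightRays and an ArrowTrail ribbon follow a shared system tip position.
# The ribbon is rebuilt each frame below the tip rather than left behind by motion.
# BowArrowSparkles spawns in two groups on ring and cylinder shapes, ramping up toward the shot, then gets a vertical force kick.
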NiagaraSystem /BRCosmetics/Animation/Game/MainPlayer/Emotes/SeleneCobra/FX/NS_Emote_SeleneCobra_BowAndArrow
dependencies {
 usedDataInterfaces [
  &NiagaraDataInterfaceCurve
  &NiagaraDataInterfaceSpriteRendererInfo
  &NiagaraDataInterfaceVector2DCurve
  &NiagaraDataInterfaceColorCurve
  &NiagaraDataInterfaceSkeletalMesh
 ]
 usedDynamicInputs [
  &Frac_Float
  &MakeVector2D
  &ClampFloat
  &Add_Vector
 ]
 usedModules [
  &self:SystemSpawnScript.SystemScriptSource.SystemScriptGraph.NiagaraNodeAssignment_2.SetVariables_#1
  &SystemState
  &self:SystemSpawnScript.SystemScriptSource.SystemScriptGraph.NiagaraNodeAssignment_1.SetVariables_#2
  &self:SystemSpawnScript.SystemScriptSource.SystemScriptGraph.NiagaraNodeAssignment_0.SetVariables_#3
 ]
}
emitters {
 ArrowTip {
  emitterUpdateScript [
   &EmitterState()
   &SpawnBurst_Instantaneous(Spawn Count=5, Use Spawn Probability=false, UseLoopCountLimit=false)
  ]
  particleSpawnScript [
   &InitializeParticle(
    Color=<rgba 0 0.25 3 1>
    Lifetime=$User.ArrowExplodeTime
    Sprite Size Mode=::NewEnumerator3
    Uniform Sprite Size=44
    UsePositionOffset=false
    Write Material Random=true
   )
   &CameraOffset(Camera Offset Amount=33)
   /SetVariables(Particles.FlareVisibilityTag:NiagaraInt32=&MakeIntFromFloat(Float=&ReturnNormalizedExecIndex(Normalized Index Includes One=false), Conversion Mode=::NewEnumerator2)) {isSetParameters=true}
  ]
  particleUpdateScript [
   &ParticleState()
   &SpriteRotationRate(Rotation Rate=&RandomRangeFloat(Minimum=-333, Maximum=333))
   &ScaleSpriteSize(
    Uniform Curve Index=&Divide_Float(A=$System.Age, B=$User.ArrowShootTime)
    Uniform Curve Scale=&FloatFromCurve(
     FloatCurve=DataInterface() {propertyValues {Curve {keys [<at 0 1>, {arriveTangent=-0.0405, leaveTangent=-0.0405, time=0.666, value=1}, {arriveTangent=-8.0089, leaveTangent=-8.0089, time=1, value=0}]}}}
     CurveIndex=$System.ArrowTipTimeline
    )
    Uniform Curve Sprite Scale=DataInterface() {propertyValues {Curve {keys [{interpMode=RCIM_Cubic, tangentMode=RCTM_User, time=0.2, value=0}, {interpMode=RCIM_Cubic, tangentMode=RCTM_User, time=1, value=1}]}}}
   )
   /SetVariables(Particles.Position=$System.ArrowTipPosition) {isSetParameters=true}
   /SetVariables(Particles.LightColor=&MultiplyLinearColor(
     Linear Color=<rgba 0 0 3 1>
     Scale Factor=&MakeVector4_FromFloat(Float=&FloatFromCurve(
       FloatCurve=DataInterface() {propertyValues {Curve {keys [{interpMode=RCIM_Cubic, tangentMode=RCTM_User, time=0.2, value=0}, {interpMode=RCIM_Cubic, tangentMode=RCTM_User, time=1, value=1}]}}}
       CurveIndex=&Divide_Float(A=$System.Age, B=$User.ArrowShootTime)
       Scale Curve=&FloatFromCurve(
        FloatCurve=DataInterface() {propertyValues {Curve {keys [<at 0 1>, {arriveTangent=-0.0405, leaveTangent=-0.0405, time=0.5, value=1}, {arriveTangent=-8.0089, leaveTangent=-8.0089, time=1, value=0}]}}}
        CurveIndex=$System.ArrowTipTimeline
        Scale Curve=6
       )
      ))
    )) {isSetParameters=true}
  ]
  rendererClasses [
   &NiagaraSpriteRendererProperties
   &NiagaraLightRendererProperties
   &NiagaraComponentRendererProperties
  ]
  renderers [
   {rendererClass=&NiagaraSpriteRendererProperties}
   {rendererClass=&NiagaraLightRendererProperties, rendererIndex=1}
   {rendererClass=&NiagaraComponentRendererProperties, rendererIndex=2}
  ]
 }
 ArrowTrail {
  emitterUpdateScript [
   &EmitterState()
   &SpawnBurst_Instantaneous(Spawn Count=6, Use Spawn Probability=false, UseLoopCountLimit=false)
  ]
  particleSpawnScript [&InitializeParticle(
    Color=&MakeLinearColorFromVectorAndFloat(
     "Vector (RGB)"=<3 3 3>
     "Float (Alpha)"=&FloatFromCurve(
      FloatCurve=DataInterface() {propertyValues {Curve {keys [
          {interpMode=RCIM_Cubic, tangentMode=RCTM_Break, time=0, value=0}
          {arriveTangent=3.6584, interpMode=RCIM_Cubic, tangentMode=RCTM_Break, time=0.96, value=1}
          <at 1 0 RCIM_Cubic>
         ]}}}
      CurveIndex=&ReturnNormalizedExecIndex()
      Scale Curve=0.5
     )
    )
    Lifetime=3
    Position Mode=::NewEnumerator0
    Ribbon Width=&FloatFromCurve(
     FloatCurve=DataInterface() {propertyValues {Curve {keys [
         {interpMode=RCIM_Cubic, leaveTangent=-1.1894, tangentMode=RCTM_Break, time=0, value=1}
         {interpMode=RCIM_Cubic, tangentMode=RCTM_Break, time=1, value=0.25}
        ]}}}
     CurveIndex=&ReturnNormalizedExecIndex()
     Scale Curve=22
    )
    Ribbon Width Mode=::NewEnumerator1
    Write Material Random=true
   )]
  particleUpdateScript [
   &ParticleState()
   &ScaleColor(
    Scale Alpha=&FloatFromCurve(
     FloatCurve=DataInterface() {propertyValues {Curve {keys [{tangentMode=RCTM_User, time=0, value=0}, <at 0.5 0.3>, <at 0.586 1>, {tangentMode=RCTM_User, time=1, value=1}]}}}
     CurveIndex=$Engine.System.Age
    )
    ScaleA=true
    ScaleRGB=true
   )
   /SetVariables(
    Particles.Position=&TransformPosition(
     Position=&AddVectorToPosition(
      Position=$System.ArrowTipPosition
      Vector=&Multiply_VectorByFloat(Vector=<0 0 -100>, Float=&OneMinusFloat(Float=&ReturnNormalizedExecIndex()))
     )
     Source Space=::Local
     Destination Space=::Simulation
    )
    Particles.RibbonU0Override=&Add_Float(
     A=&ReturnNormalizedExecIndex()
     B=&FloatFromCurve(
      FloatCurve=DataInterface() {propertyValues {Curve {keys [{tangentMode=RCTM_User, time=0, value=0}, {tangentMode=RCTM_User, time=1, value=1}]}}}
      CurveIndex=&Max_Float(A:NiagaraFloat=&Subtract_Float(A=$System.Age, B=$User.ArrowShootTime), B:NiagaraFloat=0)
      Scale Curve=5
     )
    )
   ) {isSetParameters=true}
   &ScaleRibbonWidth(Ribbon Width Scale=&FloatFromCurve(
     FloatCurve=DataInterface() {propertyValues {Curve {keys [
         {interpMode=RCIM_Cubic, tangentMode=RCTM_Break, time=0, value=0}
         {arriveTangent=1.197, interpMode=RCIM_Cubic, leaveTangent=3.467, tangentMode=RCTM_Break, time=0.5, value=0.333}
         <at 0.586 1 RCIM_Constant>
         <at 0.75 0 RCIM_Constant>
         {tangentMode=RCTM_User, time=1, value=0}
        ]}}}
     CurveIndex=$Engine.System.Age
    ))
  ]
  rendererClasses [&NiagaraRibbonRendererProperties]
  renderers [{rendererClass=&NiagaraRibbonRendererProperties}]
 }
 BowArrowSparkles {
  emitterUpdateScript [
   &EmitterState(
    Inactive Response=::NewEnumerator1
    Life Cycle Mode=::NewEnumerator1
    Loop Behavior=::NewEnumerator1
    Loop Duration=&Add_Float(A=$User.ArrowShootTime, B=0.1)
    UseLoopDelay=false
   )
   &SpawnRate(
    SpawnRate=&FloatFromCurve(
     FloatCurve=DataInterface() {propertyValues {Curve {keys [{tangentMode=RCTM_User, time=0, value=0}, {tangentMode=RCTM_User, time=1, value=1}]}}}
     CurveIndex=&Divide_Float(A=$System.Age, B=$User.ArrowShootTime)
     Scale Curve=222
    )
    Use Spawn Probability=false
   )
   &SpawnRate(
    SpawnGroup=1
    SpawnRate=&FloatFromCurve(
     FloatCurve=DataInterface() {propertyValues {Curve {keys [{tangentMode=RCTM_User, time=0, value=0}, {tangentMode=RCTM_User, time=1, value=1}]}}}
     CurveIndex=&Divide_Float(A=$System.Age, B=$User.ArrowShootTime)
     Scale Curve=66
    )
    Use Spawn Probability=false
   ) {module=SpawnRate001}
  ]
  particleSpawnScript [
   &InitializeParticle(
    AdjustAlpha=false
    AdjustHue=true
    AdjustSaturation=false
    AdjustValue=false
    Color=<rgba 0 0.0833 1 1>
    Color Mode=::NewEnumerator3
    Lifetime=&MakeCustomFloatFromBool(
     Bool=&SetBoolByFloatComparison(A=$System.Age, B=$User.ArrowShootTime, Comparison Type=::NewEnumerator6)
     True Float=&RandomRangeFloat(Minimum=0.233, Maximum=0.333)
     False Float=&RandomRangeFloat(Minimum=0.333, Maximum=1)
    )
    Sprite Size Mode=::NewEnumerator3
    Uniform Sprite Size=&RandomRangeFloat(Minimum=6, Maximum=22)
    UsePositionOffset=false
    Write Material Random=true
   )
   &ShapeLocation(
    Additional Quaternion Rotation=false
    MaskSpawnGroup=true
    Offset=<-10 147.5 -15>
    Offset Mode=::NewEnumerator0
    Override Seed=false
    Ring / Disc Distribution Mode:ENiagaraSphereDistributionMode=::NewEnumerator1
    Ring Radius=50
    Rotation Mode=::NewEnumerator3
    Rotation Quaternion=&MultiplyQuaternion(
     Quaternion A=$Engine.Owner.Rotation
     Quaternion B=&MakeQuatFromAxisAngle(Axis=<1 0.07 0>, AngleInDegrees=90, NormalizeAxis=true)
    )
    Shape Primitive=::NewEnumerator4
    Transform Order:ENiagara_TransformOrder=::NewEnumerator1
    U Position=&RandomRangeFloat(Minimum=0, Maximum=0.5)
    UseOwnerScale=false
   )
   &ShapeLocation(
    Cylinder Height=65
    Cylinder Height Midpoint=0
    Cylinder Radius=0
    MaskSpawnGroup=true
    Offset=<6 13 140>
    Offset Coordinate Space=::Local
    Offset Mode=::NewEnumerator0
    Override Seed=false
    Shape Primitive=::NewEnumerator1
    Spawn Group Mask=1
    Surface Only=false
    UseOwnerScale=true
   ) {module=ShapeLocation001}
   &ShapeLocation(MaskSpawnGroup=false, Override Seed=false, Sphere Radius=3, UseOwnerScale=true) {module=ShapeLocation002}
   &CameraOffset(Camera Offset Amount=22)
  ]
  particleUpdateScript [
   &ParticleState()
   &LinearForce(Force=&MakeVector(
     X=0
     Y=0
     Z=&FloatFromCurve(
      FloatCurve=DataInterface() {propertyValues {Curve {keys [<at -0.1 0 RCIM_Constant>, {interpMode=RCIM_Cubic, leaveTangent=-9.804, tangentMode=RCTM_Break, time=0, value=1}, <at 0.25 0>]}}}
      CurveIndex=&Subtract_Float(A=$System.Age, B=$User.ArrowShootTime)
      Scale Curve=2222
     )
    ))
   &Drag(Drag=2, Use Linear Drag=true, Use Rotational Drag=true)
   &SolveForcesAndVelocity(Clamp Velocity=false, Limit Acceleration=false)
   &ScaleColor(
    Scale Alpha=&FloatFromCurve(
     FloatCurve=DataInterface() {propertyValues {Curve {keys [
         {arriveTangent=9.8157, interpMode=RCIM_Cubic, leaveTangent=9.8157, tangentMode=RCTM_User, time=0, value=0}
         {arriveTangent=-1.4617, interpMode=RCIM_Cubic, leaveTangent=-1.4617, tangentMode=RCTM_User, time=0.2, value=1}
         {interpMode=RCIM_Cubic, tangentMode=RCTM_User, time=0.3, value=0}
         {interpMode=RCIM_Cubic, tangentMode=RCTM_User, time=0.5, value=0.75}
         {interpMode=RCIM_Cubic, tangentMode=RCTM_Break, time=1, value=0}
        ]}}}
     Scale Curve=&Waveform(
      Global Phase=&Add_Float(A=$System.Age, B=&RandomRangeFloat(Minimum=0, Maximum=1))
      "Waveform [1]"=::NewEnumerator0
      "[1] Frequency"=1
      "[1] Amplitude Min/Max"=<0 1>
     )
    )
    ScaleA=true
    ScaleRGB=true
   )
  ]
  rendererClasses [&NiagaraSpriteRendererProperties]
  renderers [
   {rendererClass=&NiagaraSpriteRendererProperties}
   {rendererClass=&NiagaraSpriteRendererProperties, rendererIndex=1}
  ]
 }
 LightRays {
  emitterUpdateScript [
   &EmitterState()
   &SpawnBurst_Instantaneous(Spawn Count=9, Use Spawn Probability=false, UseLoopCountLimit=false)
  ]
  particleSpawnScript [
   &InitializeParticle(
    Color=<rgba 0 0.2667 1 0.3>
    Lifetime=$User.ArrowExplodeTime
    Position Offset=$System.ArrowTipPosition
    Position Offset Coordinate Space=::Local
    Sprite Rotation Mode=::NewEnumerator2
    Sprite Size Max=<11 55>
    Sprite Size Min=<5 33>
    Sprite Size Mode=::NewEnumerator2
    UsePositionOffset=true
    Write Material Random=true
   )
   &SubUVAnimation(
    End Frame Range Override=3
    Random Start Frame=true
    "Sprite Renderer "=DataInterface() {propertyValues {SpriteRenderer=&self:LightRays_9.NiagaraSpriteRendererProperties_0}}
    UseEndFrame=false
    UseStartFrame=false
   )
  ]
  particleUpdateScript [
   &ParticleState()
   &SpriteRotationRate(Rotation Rate=&RandomRangeFloat(Minimum=-233, Maximum=233))
   &ScaleSpriteSize(
    Non-Uniform Curve Index:NiagaraFloat=&Divide_Float(A=$System.Age, B=$User.ArrowShootTime)
    Non-Uniform Curve Sprite Scale=DataInterface() {propertyValues {
      XCurve {keys [{tangentMode=RCTM_User, time=0, value=1}, <at 1 1>, <at 1.2 0>]}
      YCurve {keys [
        {interpMode=RCIM_Cubic, tangentMode=RCTM_Break, time=0, value=0}
        {arriveTangent=0.0267, tangentMode=RCTM_User, time=1, value=1}
        <at 1.2 0.5>
       ]}
     }}
    Scale Sprite Size Mode=::NewEnumerator3
   )
   &ScaleColor(
    Scale Alpha=&FloatFromCurve(
     FloatCurve=DataInterface() {propertyValues {Curve {keys [
         {tangentMode=RCTM_User, time=0, value=1}
         {arriveTangent=-0.0405, interpMode=RCIM_Cubic, leaveTangent=-0.0405, tangentMode=RCTM_User, time=1, value=1}
         {arriveTangent=-8.0089, interpMode=RCIM_Cubic, leaveTangent=-8.0089, tangentMode=RCTM_User, time=1.3, value=0}
        ]}}}
     CurveIndex=&Divide_Float(A=$System.Age, B=$User.ArrowShootTime)
    )
    Scale RGB=&VectorFromFloat(Value=&FloatFromCurve(
      FloatCurve=DataInterface() {propertyValues {Curve {keys [<at 0 5>, {arriveTangent=-92.867, time=0.333, value=1}, <at 1 3>]}}}
      CurveIndex=&Divide_Float(A=$System.Age, B=$User.ArrowShootTime)
     ))
    ScaleA=false
    ScaleRGB=true
   )
   /SetVariables(Particles.Position:Vector3f=$System.ArrowTipPosition) {isSetParameters=true}
  ]
  rendererClasses [&NiagaraSpriteRendererProperties]
  renderers [{rendererClass=&NiagaraSpriteRendererProperties}]
 }
}
summary {
 systemName=NS_Emote_SeleneCobra_BowAndArrow
 emitters [
  {
   emitterName=ArrowTip
   rendererClasses [
    &NiagaraSpriteRendererProperties
    &NiagaraLightRendererProperties
    &NiagaraComponentRendererProperties
   ]
  }
  {emitterName=BowArrowSparkles, rendererClasses [&NiagaraSpriteRendererProperties]}
  {emitterName=ArrowTrail, rendererClasses [&NiagaraRibbonRendererProperties]}
  {emitterName=LightRays, rendererClasses [&NiagaraSpriteRendererProperties]}
 ]
 userVariables [
  User.ArrowExplodeTime {type=NiagaraFloat, defaultValue=0.733, description=""}
  User.ArrowShootTime {type=NiagaraFloat, defaultValue=0.6, description=""}
  User.ArrowTipStartingOffset {type=Vector3f, description="", defaultValue {struct=Vector3f, value=<5 13 214>}}
 ]
}
