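# Niagara system for the loop part of the SeleneCobra emote, producing a mostly blue and purple burst at an arrow tip.
# Burst emitters layer flare, glow, halo, glints, rays, dust, meshes and a shockwave at System.ArrowTipPosition, plus a timeline-scaled explosion light.
# Separate emitters add flashing sparkles and foot dust spawned from skeletal mesh foot height and velocity.
# A shared time-of-day multiplier scales brightness, and some values switch between gameplay and front end.
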NiagaraSystem /BRCosmetics/Animation/Game/MainPlayer/Emotes/SeleneCobra/FX/NS_Emote_SeleneCobra_Loop
dependencies {
 usedDataInterfaces [
  &NiagaraDataInterfaceCurve
  &NiagaraDataInterfaceColorCurve
  &NiagaraDataInterfaceParticleRead
  &NiagaraDataInterfaceSpriteRendererInfo
  &NiagaraDataInterfaceVector2DCurve
  &NiagaraDataInterfaceSkeletalMesh
 ]
 usedDynamicInputs [
  &MakeVector2D
  &Add_Vector
  /Game/Effects/Niagara/DynamicInputs/State/FN_IsExecutionStateActive
  &FNTimeOfDayScalar
 ]
 usedModules [
  &self:SystemSpawnScript.SystemScriptSource.SystemScriptGraph.NiagaraNodeAssignment_2.SetVariables_#1
  &SystemState
  &self:SystemSpawnScript.SystemScriptSource.SystemScriptGraph.NiagaraNodeAssignment_0.SetVariables_#2
  &Timeline
  &self:SystemSpawnScript.SystemScriptSource.SystemScriptGraph.NiagaraNodeAssignment_1.SetVariables_#3
 ]
}
emitters {
 ArrowExplosionLight {
  emitterUpdateScript [
   &EmitterState()
   &SpawnBurst_Instantaneous(Use Spawn Probability=false, UseLoopCountLimit=false)
  ]
  particleSpawnScript [&InitializeParticle(
    Color=<rgba 0 0.0833 1 1>
    Lifetime=13
    Position Offset=&AddVectorToPosition(
     Position=$System.ArrowTipPosition
     Vector=&FN_FrontEndVector3Modifier(Gameplay=<0 0 0>, Front End=<0 0 0>)
    )
    Position Offset Coordinate Space=::Local
    Sprite Size Mode=::NewEnumerator3
    Uniform Sprite Size=444
    UsePositionOffset=true
    Write Material Random=true
   )]
  particleUpdateScript [
   &ParticleState()
   &ScaleSpriteSize(
    Uniform Curve Index=$System.Age
    Uniform Curve Sprite Scale=DataInterface() {propertyValues {Curve {keys [
        <at 0 0>
        {arriveTangent=1.1179, interpMode=RCIM_Cubic, leaveTangent=1.1179, tangentMode=RCTM_User, time=0.1, value=0.6228}
        {arriveTangent=0.0224, interpMode=RCIM_Constant, time=1, value=1}
        <at 2 0 RCIM_Cubic>
       ]}}}
   )
   &ScaleColor(
    Scale Alpha=&FloatFromCurve(
     FloatCurve=DataInterface() {propertyValues {Curve {keys [<at 0 0>, {tangentMode=RCTM_User, time=0.1, value=1}, <at 0.2 0.25>, {tangentMode=RCTM_User, time=2, value=0}]}}}
     CurveIndex=$System.Age
     Scale Curve=$System.Timeline.Time
    )
    Scale RGB=&VectorFromFloat(Value=&FloatFromCurve(
      FloatCurve=DataInterface() {propertyValues {Curve {keys [{tangentMode=RCTM_User, time=0.1, value=1}, <at 0.2 0.5>, <at 2 0>]}}}
      CurveIndex=$System.Age
      Scale Curve=&Multiply_Float(A=44, B=$System.TimeOfDayBrightnessMultiplier)
     ))
    ScaleA=true
    ScaleRGB=true
   )
   /SetVariables(
    Particles.ComponentLightIntensity:NiagaraFloat=&FloatFromCurve(
     FloatCurve=DataInterface() {propertyValues {Curve {keys [
         <at 0 0>
         {leaveTangent=-1.7408, time=0.1, value=3}
         {arriveTangent=-0.5085, leaveTangent=-0.5085, time=1.5, value=1.25}
         <at 9 1>
         <at 13 0>
        ]}}}
     CurveIndex=$System.Age
     Scale Curve=&Multiply_Float(
      A=$System.Timeline.Time
      B=&Multiply_Float(A=&FN_FrontEndFloatModifier(Gameplay=66, Front End=50), B=$System.TimeOfDayBrightnessMultiplier)
     )
    )
    Particles.ComponentLightRadius:NiagaraFloat=&FloatFromCurve(
     FloatCurve=DataInterface() {propertyValues {Curve {keys [<at 0 0>, {leaveTangent=-1.7408, time=0.1, value=5}, <at 9 4>, <at 13 0>]}}}
     CurveIndex=$System.Age
     Scale Curve=68
    )
    Particles.LightColor=&MultiplyLinearColor(
     Linear Color=$Particles.Initial.Color
     Scale Factor=&MakeFloat4FromVectorAndFloat(
      "Vector (RGB)"=&VectorFromFloat(Value=&Multiply_Float(A=66, B=$System.TimeOfDayBrightnessMultiplier))
      "Float (Alpha)"=&FloatFromCurve(
       FloatCurve=DataInterface() {propertyValues {Curve {keys [
           <at 0 0>
           {leaveTangent=-1.7408, time=0.1, value=3}
           {arriveTangent=-0.5085, leaveTangent=-0.5085, time=1.5, value=1.25}
           <at 9 1>
           <at 13 0>
          ]}}}
       CurveIndex=$System.Age
       Scale Curve=$System.Timeline.Time
      )
     )
    )
    Particles.LightRadius=&FloatFromCurve(
     FloatCurve=DataInterface() {propertyValues {Curve {keys [<at 0 0>, {leaveTangent=-1.7408, time=0.1, value=5}, {arriveTangent=-0.5085, leaveTangent=-0.5085, time=1.5, value=4}, <at 9 3.5>, <at 13 0>]}}}
     CurveIndex=$System.Age
     Scale Curve=66
    )
   ) {isSetParameters=true}
   &KillParticles(Kill Particles=&SetBoolByFloatComparison(A=$System.Timeline.Time, B=0, Comparison Type=::NewEnumerator7))
  ]
  rendererClasses [
   &NiagaraSpriteRendererProperties
   &NiagaraLightRendererProperties
   &NiagaraComponentRendererProperties
  ]
  renderers [
   {rendererClass=&NiagaraSpriteRendererProperties}
   {rendererClass=&NiagaraLightRendererProperties, rendererIndex=1}
   {rendererClass=&NiagaraComponentRendererProperties, rendererIndex=2}
  ]
 }
 BG_Glow {
  emitterUpdateScript [
   &EmitterState(Life Cycle Mode=::NewEnumerator1, Loop Behavior=::NewEnumerator1, Loop Delay=0.05, Loop Duration=1, UseLoopDelay=true)
   &SpawnBurst_Instantaneous(Spawn Count=1, Use Spawn Probability=false)
  ]
  particleSpawnScript [
   &InitializeParticle(
    Color=<rgba 0.1769 0.0064 0.3073 0.5>
    Lifetime=0.5
    Position Offset=$System.ArrowTipPosition
    Position Offset Coordinate Space=::Local
    Sprite Size Mode=::NewEnumerator3
    Uniform Sprite Size=333
    UsePositionOffset=true
    Write Material Random=true
   )
   &CameraOffset(Camera Offset Amount=-20)
  ]
  particleUpdateScript [
   &ParticleState()
   &ScaleSpriteSize(Scale Factor=&Vector2DFromFloat(Value=&FloatFromCurve(FloatCurve=DataInterface() {propertyValues {Curve {keys [
          {arriveTangent=0.4346, interpMode=RCIM_Cubic, leaveTangent=0.4346, tangentMode=RCTM_User, time=0, value=0}
          {arriveTangent=0.0473, interpMode=RCIM_Cubic, leaveTangent=0.0473, tangentMode=RCTM_User, time=0.2, value=1}
         ]}}})))
   &ScaleColor(
    Scale Alpha=&FloatFromCurve(
     FloatCurve=DataInterface() {propertyValues {Curve {keys [
         {arriveTangent=0.4005, interpMode=RCIM_Cubic, leaveTangent=0.4005, tangentMode=RCTM_User, time=0, value=0}
         {arriveTangent=0.0256, interpMode=RCIM_Cubic, leaveTangent=0.0256, tangentMode=RCTM_User, time=0.2, value=1}
         <at 1 0 RCIM_Cubic>
        ]}}}
     Scale Curve=$System.TimeOfDayBrightnessMultiplier
    )
    ScaleA=true
    ScaleRGB=true
   )
   /SetVariables(
    Particles.GlowColor2:LinearColor=&MakeLinearColorFromVectorAndFloat(
     "Vector (RGB)"=&MakeVectorFromLinearColorRGB(Color=<rgba 0.4427 0.6706 5 1>)
     "Float (Alpha)"=&Multiply_Float(
      A=0.2
      B=&FloatFromCurve(FloatCurve=DataInterface() {propertyValues {Curve {keys [
           {arriveTangent=0.4005, interpMode=RCIM_Cubic, leaveTangent=0.4005, tangentMode=RCTM_User, time=0, value=0}
           {arriveTangent=0.0256, interpMode=RCIM_Cubic, leaveTangent=0.0256, tangentMode=RCTM_User, time=0.2, value=1}
           <at 1 0 RCIM_Cubic>
          ]}}})
     )
    )
    Particles.GlowSize2:Vector2f=&Vector2DFromFloat(Value=&Multiply_Float(
      A=120
      B=&FloatFromCurve(FloatCurve=DataInterface() {propertyValues {Curve {keys [
           {arriveTangent=0.4346, interpMode=RCIM_Cubic, leaveTangent=0.4346, tangentMode=RCTM_User, time=0, value=0}
           {arriveTangent=0.0473, interpMode=RCIM_Cubic, leaveTangent=0.0473, tangentMode=RCTM_User, time=0.2, value=1}
          ]}}})
     ))
   ) {isSetParameters=true}
  ]
  rendererClasses [&NiagaraSpriteRendererProperties]
  renderers [
   {rendererClass=&NiagaraSpriteRendererProperties}
   {rendererClass=&NiagaraSpriteRendererProperties, rendererIndex=1}
  ]
 }
 Dust {
  emitterUpdateScript [
   &EmitterState(
    Enable Distance Culling=true
    MaxDistance=10000
    Module.bUseMaxDistance=true
    Module.bUseMinDistance=false
    Scalability Mode=::NewEnumerator1
   )
   &SpawnBurst_Instantaneous(Spawn Count=8, Spawn Time=0.1, Use Spawn Probability=false, UseLoopCountLimit=false)
  ]
  particleSpawnScript [
   &InitializeParticle(
    Color Mode=::NewEnumerator0
    Lifetime Max=2.6
    Lifetime Min=0.7
    Lifetime Mode=::NewEnumerator1
    Position Offset=$System.ArrowTipPosition
    Position Offset Coordinate Space=::Local
    Sprite Rotation Mode=::NewEnumerator2
    Sprite Size Mode=::NewEnumerator1
    Uniform Sprite Size Max=120
    Uniform Sprite Size Min=80
    UsePositionOffset=true
    Write Material Random=true
   )
   &ShapeLocation(MaskSpawnGroup=false, Offset Mode=::NewEnumerator0, Override Seed=false, Sphere Radius=10, UseOwnerScale=true)
   &AddVelocity(
    Module.Use Falloff Distance=false
    Origin Offset=$System.ArrowTipPosition
    Override Seed=false
    Velocity Mode=::NewEnumerator1
    Velocity Speed=&RandomRangeFloat(Minimum=333, Maximum=555)
   )
   &CameraOffset(Camera Offset Amount=&RandomRangeFloat(Minimum=-10, Maximum=-50, Evaluation Type=::NewEnumerator1))
  ]
  particleUpdateScript [
   &ParticleState()
   &ScaleSpriteSize(
    Scale Sprite Size Mode=::NewEnumerator0
    Uniform Scale Factor=&FloatFromCurve(
     FloatCurve=DataInterface() {propertyValues {Curve {keys [
         {interpMode=RCIM_Cubic, tangentMode=RCTM_User, time=0, value=0.3}
         {arriveTangent=0.5158, interpMode=RCIM_Cubic, leaveTangent=0.5158, tangentMode=RCTM_User, time=0.4, value=0.9}
         {arriveTangentWeight=0.3824, interpMode=RCIM_Cubic, tangentMode=RCTM_User, time=1, value=1}
        ]}}}
     Scale Curve=1.5
    )
   )
   &Drag(Drag=16, Use Linear Drag=true, Use Rotational Drag=true)
   &GravityForce(Gravity=<0 0 -333>)
   &SolveForcesAndVelocity(Clamp Velocity=false, Limit Acceleration=false)
   &SpriteRotationRate(Rotation Rate=&RandomRangeFloat(Minimum=-0.03, Maximum=0.03))
   &ScaleColor(
    Scale Alpha=&FloatFromCurve(
     FloatCurve=DataInterface() {propertyValues {Curve {keys [
         {arriveTangent=9.0725, interpMode=RCIM_Cubic, leaveTangent=9.0725, tangentMode=RCTM_User, time=0, value=0}
         {arriveTangent=-2.5249, interpMode=RCIM_Cubic, leaveTangent=-2.5249, tangentMode=RCTM_User, time=0.1408, value=0.9847}
         {arriveTangent=-0.0405, interpMode=RCIM_Cubic, leaveTangent=-0.0405, tangentMode=RCTM_User, time=1, value=0}
        ]}}}
     Scale Curve=0.05
    )
    Scale RGB=&VectorFromFloat(Value=$System.TimeOfDayBrightnessMultiplier)
    ScaleA=true
    ScaleRGB=true
   )
   &DynamicMaterialParameters(
    Index 0 Param 1=&Add_Float(
     A=&RandomRangeFloat(Minimum=0, Maximum=1)
     B=&FloatFromCurve(
      FloatCurve=DataInterface() {propertyValues {Curve {keys [
          {arriveTangent=2.7533, interpMode=RCIM_Cubic, leaveTangent=1.8958, tangentMode=RCTM_User, time=0, value=1}
          {arriveTangent=0.8292, interpMode=RCIM_Cubic, leaveTangent=0.8292, tangentMode=RCTM_User, time=0.321, value=1.3453}
          {arriveTangent=0.0529, interpMode=RCIM_Cubic, leaveTangent=0.2021, tangentMode=RCTM_User, time=1, value=1.5}
         ]}}}
      CurveIndex=&Add_Float(A=&Multiply_Float(A=$Particles.Initial.MaterialRandom, B=0.5), B=$Particles.NormalizedAge)
      Scale Curve=&Lerp_Float(A=1, B=2, Alpha=0.5)
     )
    )
    Index 0 Param 2=&FloatFromCurve(FloatCurve=DataInterface() {propertyValues {Curve {keys [<at 0 1>,<at 1 0>]}}})
    Index 0 Param 3=&FloatFromCurve(
     FloatCurve=DataInterface() {propertyValues {Curve {keys [
         {arriveTangent=-2.2052, interpMode=RCIM_Cubic, leaveTangent=-2.2052, tangentMode=RCTM_User, time=0.1, value=1}
         {interpMode=RCIM_Cubic, tangentMode=RCTM_User, time=1, value=0.4}
        ]}}}
     Scale Curve=&Multiply_Float(A=0.5, B=$System.TimeOfDayBrightnessMultiplier)
    )
    Index 0 Param 4=&FloatFromCurve(FloatCurve=DataInterface() {propertyValues {Curve {keys [{interpMode=RCIM_Cubic, tangentMode=RCTM_Break, time=0.5, value=1}, {interpMode=RCIM_Cubic, tangentMode=RCTM_Break, time=1, value=0}]}}})
    Index 1 Param 1=&FloatFromCurve(
     FloatCurve=DataInterface() {propertyValues {Curve {keys [
         {interpMode=RCIM_Cubic, leaveTangentWeight=0.0833, tangentMode=RCTM_Break, time=0, value=0}
         {arriveTangentWeight=0.0873, interpMode=RCIM_Cubic, tangentMode=RCTM_Break, time=0.25, value=1}
        ]}}}
     Scale Curve=&Lerp_Float(A=0.35, B=0.6, Alpha=0.5)
    )
    Index 1 Param 2=&FloatFromCurve(
     FloatCurve=DataInterface() {propertyValues {Curve {keys [
         {interpMode=RCIM_Cubic, tangentMode=RCTM_Break, time=0, value=0.2}
         {interpMode=RCIM_Cubic, leaveTangent=0.9066, tangentMode=RCTM_Break, time=1, value=2}
        ]}}}
     Scale Curve=&RandomRangeFloat(Minimum=0.1, Maximum=0.2)
    )
    Index 1 Param 3=&FloatFromCurve(FloatCurve=DataInterface() {propertyValues {Curve {keys [
         {interpMode=RCIM_Cubic, leaveTangent=6.3917, tangentMode=RCTM_Break, time=0, value=0}
         {arriveTangent=1.7549, interpMode=RCIM_Cubic, leaveTangent=10.6767, tangentMode=RCTM_User, time=0.25, value=1}
        ]}}})
    Index 1 Param 4=&FloatFromCurve(
     FloatCurve=DataInterface() {propertyValues {Curve {keys [{arriveTangent=-0.136, interpMode=RCIM_Cubic, leaveTangent=-0.136, tangentMode=RCTM_User, time=0, value=1}, <at 1 0.95>]}}}
     Scale Curve=&Lerp_Float(A=2, B=5, Alpha=0.5)
    )
    Index 2 Param 1=&FloatFromCurve(FloatCurve=DataInterface() {propertyValues {Curve {keys [
         {interpMode=RCIM_Cubic, leaveTangent=4.2485, tangentMode=RCTM_Break, time=0, value=0}
         {arriveTangent=1.0548, interpMode=RCIM_Cubic, leaveTangent=1.0548, tangentMode=RCTM_User, time=0.17, value=0.5715}
         <at 1 1>
        ]}}})
    Index 2 Param 2=&FloatFromCurve(
     FloatCurve=DataInterface() {propertyValues {Curve {keys [
         {arriveTangent=1.9056, interpMode=RCIM_Cubic, leaveTangent=1.9056, tangentMode=RCTM_User, time=0, value=0.2}
         {arriveTangent=0.0259, interpMode=RCIM_Cubic, leaveTangent=0.0259, tangentMode=RCTM_User, time=1, value=1.4}
        ]}}}
     Scale Curve=1.5
    )
    Index 2 Param 3=&FloatFromCurve(
     FloatCurve=DataInterface() {propertyValues {Curve {keys [
         {interpMode=RCIM_Cubic, leaveTangentWeight=0.1411, tangentMode=RCTM_Break, time=0, value=2}
         {interpMode=RCIM_Cubic, tangentMode=RCTM_Break, time=0.3, value=1}
         {interpMode=RCIM_Cubic, leaveTangent=11.1233, tangentMode=RCTM_Break, time=1, value=2}
        ]}}}
     Scale Curve=10
    )
    Index 2 Param 4=&FloatFromCurve(
     FloatCurve=DataInterface() {propertyValues {Curve {keys [
         {arriveTangent=-0.6867, interpMode=RCIM_Cubic, leaveTangent=-0.6867, tangentMode=RCTM_User, time=0, value=1}
         {interpMode=RCIM_Cubic, leaveTangent=-1.2956, tangentMode=RCTM_Break, time=1, value=0.75}
        ]}}}
     Scale Curve=&RandomRangeFloat(Minimum=0.5, Maximum=2)
    )
    Module.Param0WriteEnabled=true
    Module.Param0WriteEnabled1=true
    Module.Param0WriteEnabled2=true
    Module.Param1WriteEnabled=false
    Module.Param1WriteEnabled1=true
    Module.Param1WriteEnabled2=true
    Module.Param2WriteEnabled=true
    Module.Param2WriteEnabled1=true
    Module.Param2WriteEnabled2=true
    Module.Param3WriteEnabled=true
    Module.Param3WriteEnabled1=true
    Module.Param3WriteEnabled2=true
    Write Parameter Index 1=true
    Write Parameter Index 2=true
   )
  ]
  rendererClasses [&NiagaraSpriteRendererProperties]
  renderers [
   {rendererClass=&NiagaraSpriteRendererProperties}
   {rendererClass=&NiagaraSpriteRendererProperties, rendererIndex=1}
  ]
 }
 DustPuffFeet {
  emitterUpdateScript [
   &EmitterState(Life Cycle Mode=::NewEnumerator1, Loop Behavior=::NewEnumerator1, Loop Delay=0.1, Loop Duration Mode=::NewEnumerator1, UseLoopDelay=false)
   /Game/Animation/Game/MainPlayer/Emotes/SeleneCobra/FX/SpawnFootDust(
    Height Threshold For Spawning=8
    Minimum Interval Between Spawn Events=0.2
    Right Foot Bone Index=1
    Skeletal Mesh:NiagaraDataInterfaceSkeletalMesh=$System.Skeletal Mesh
    Spawn Count=7
    Velocity Threshold For Spawning=5
   )
  ]
  particleSpawnScript [
   &InitializeParticle(
    Color=<rgba 0.13 0.1206 0.1092 0.166>
    Lifetime Max=1
    Lifetime Min=0.6
    Lifetime Mode=::NewEnumerator1
    Position=&MaskVectorBySpawnGroup(
     Masked Vector=Linked() {linkedVariable {name=Emitter.LeftFootSpawnPositionWorldSpace, type=NiagaraPosition}}
     Passthrough Vector=Linked() {linkedVariable {name=Emitter.RightFootSpawnPositionWorldSpace, type=NiagaraPosition}}
    )
    Position Mode=::NewEnumerator1
    Sprite Rotation Mode=::NewEnumerator2
    Sprite Size Mode=::NewEnumerator1
    Uniform Sprite Size Max=44
    Uniform Sprite Size Min=22
    UsePositionOffset=false
    Write Material Random=true
   )
   &ShapeLocation(
    MaskSpawnGroup=false
    Override Seed=false
    Ring Radius=11
    Shape Origin=&MakeVector(
     X=&MakeFloatFromVector(Vector=Linked() {linkedVariable {name=StackContext.Position, type=NiagaraPosition}}, Channel=::NewEnumerator0)
     Y=&MakeFloatFromVector(Vector=Linked() {linkedVariable {name=StackContext.Position, type=NiagaraPosition}}, Channel=::NewEnumerator1)
     Z=&Add_Float(
      A=5
      B=&MakeFloatFromVector(Vector=Linked() {linkedVariable {name=Engine.Owner.Position, type=NiagaraPosition}}, Channel=::NewEnumerator2)
     )
    )
    Shape Primitive=::NewEnumerator4
    UseOwnerScale=false
   )
   &AddVelocity(
    Velocity=$Particles.ShapeLocation.ShapeVector
    Velocity Speed Scale=&RandomRangeFloat(Minimum=1, Maximum=13)
   )
  ]
  particleUpdateScript [
   &CameraOffset(Camera Offset Amount=&FloatFromCurve(
     FloatCurve=DataInterface() {propertyValues {Curve {keys [{tangentMode=RCTM_User, time=0, value=0}, {tangentMode=RCTM_User, time=1, value=1}]}}}
     Scale Curve=44
    ))
   &Drag(Drag=9, Use Linear Drag=true, Use Rotational Drag=true)
   &GravityForce(Gravity=<0 0 111>)
   &ParticleState()
   &SolveForcesAndVelocity(Clamp Velocity=false, Limit Acceleration=false)
   &ScaleColor(
    Color Value To Scale=$Particles.Color
    Scale Alpha=&Multiply_Float(
     A=&FloatFromCurve(FloatCurve=DataInterface() {propertyValues {Curve {keys [
          {arriveTangent=9.0725, interpMode=RCIM_Cubic, leaveTangent=9.0725, tangentMode=RCTM_User, time=0, value=0}
          {arriveTangent=-2.5249, interpMode=RCIM_Cubic, leaveTangent=-2.5249, tangentMode=RCTM_User, time=0.1408, value=0.9847}
          {arriveTangent=-0.0405, interpMode=RCIM_Cubic, leaveTangent=-0.0405, tangentMode=RCTM_User, time=1, value=0}
         ]}}})
     B=$System.TimeOfDayBrightnessMultiplier
    )
    Scale RGB=&VectorFromFloat(Value=&FloatFromCurve(FloatCurve=DataInterface() {propertyValues {Curve {keys [<at 0 1>,<at 0.9983 0.5>]}}}))
    ScaleA=true
    ScaleRGB=true
   )
   &DynamicMaterialParameters(
    Index 0 Param 1=&Add_Float(
     A=$Particles.RandomFloat
     B=&FloatFromCurve(
      FloatCurve=DataInterface() {propertyValues {Curve {keys [
          {arriveTangent=0.9088, interpMode=RCIM_Cubic, leaveTangent=0.9088, tangentMode=RCTM_User, time=0, value=1}
          {arriveTangent=0.5473, interpMode=RCIM_Cubic, leaveTangent=0.5473, tangentMode=RCTM_User, time=1, value=2}
         ]}}}
      CurveIndex=$Particles.Age
      Scale Curve=0.222
     )
    )
    Index 0 Param 2=&FloatFromCurve(FloatCurve=DataInterface() {propertyValues {Curve {keys [
         {arriveTangent=0.0453, interpMode=RCIM_Cubic, tangentMode=RCTM_Break, time=0, value=-1}
         {arriveTangent=0.0604, interpMode=RCIM_Cubic, leaveTangent=0.0604, tangentMode=RCTM_User, time=1, value=-1}
        ]}}})
    Index 0 Param 3=&FloatFromCurve(FloatCurve=DataInterface() {propertyValues {Curve {keys [
         {arriveTangent=-9.0447, interpMode=RCIM_Cubic, leaveTangent=-2.0044, tangentMode=RCTM_User, time=0, value=0.5}
         {interpMode=RCIM_Cubic, leaveTangent=4.6779, time=0.2851, value=0}
        ]}}})
    Index 1 Param 1=&FloatFromCurve(
     FloatCurve=DataInterface() {propertyValues {Curve {keys [<at 0 0>,<at 0.25 1>]}}}
     Scale Curve=&RandomRangeFloat(Minimum=2, Maximum=1)
    )
    Index 1 Param 2=&FloatFromCurve(
     FloatCurve=DataInterface() {propertyValues {Curve {keys [
         {interpMode=RCIM_Cubic, leaveTangent=3.7422, tangentMode=RCTM_Break, time=0, value=0.2}
         {interpMode=RCIM_Cubic, leaveTangent=0.9066, time=1, value=2}
        ]}}}
     Scale Curve=&RandomRangeFloat(Minimum=0.5, Maximum=1)
    )
    Index 1 Param 3=&FloatFromCurve(FloatCurve=DataInterface() {propertyValues {Curve {keys [
         {interpMode=RCIM_Cubic, leaveTangent=2.8445, tangentMode=RCTM_Break, time=0, value=0.2}
         {interpMode=RCIM_Cubic, leaveTangent=63.2978, time=0.25, value=1}
        ]}}})
    Index 1 Param 4=&FloatFromCurve(
     FloatCurve=DataInterface() {propertyValues {Curve {keys [<at 0 1>,<at 1 0.8>]}}}
     Scale Curve=&RandomRangeFloat(Minimum=0.5, Maximum=1)
    )
    Index 2 Param 1=&FloatFromCurve(
     FloatCurve=DataInterface() {propertyValues {Curve {keys [{interpMode=RCIM_Cubic, leaveTangent=3.2309, tangentMode=RCTM_Break, time=0.1, value=0}, <at 1 1>]}}}
     Scale Curve=0
    )
    Index 2 Param 2=&FloatFromCurve(
     FloatCurve=DataInterface() {propertyValues {Curve {keys [<at 0 0.2 RCIM_Cubic>,<at 1 0.8 RCIM_Cubic>]}}}
     Scale Curve=&RandomRangeFloat(Minimum=0.2, Maximum=1.2)
    )
    Index 2 Param 3=&FloatFromCurve(FloatCurve=DataInterface() {propertyValues {Curve {keys [
         {interpMode=RCIM_Cubic, tangentMode=RCTM_Break, time=0, value=1}
         {interpMode=RCIM_Cubic, leaveTangent=11.1233, tangentMode=RCTM_Break, time=1, value=2}
        ]}}})
    Index 2 Param 4=&FloatFromCurve(
     FloatCurve=DataInterface() {propertyValues {Curve {keys [<at 0 1 RCIM_Cubic>, {interpMode=RCIM_Cubic, leaveTangent=-1.2956, tangentMode=RCTM_Break, time=1, value=0.75}]}}}
     Scale Curve=0.8
    )
    Module.Index0Param0Write=true
    Module.Index0Param1Write=true
    Module.Index0Param2Write=true
    Module.Index0Param3Write=true
    Module.Index1Param0Write=true
    Module.Index1Param1Write=true
    Module.Index1Param2Write=true
    Module.Index1Param3Write=true
    Module.Index2Param0Write=true
    Module.Index2Param1Write=true
    Module.Index2Param2Write=true
    Module.Index2Param3Write=true
    Write Parameter Index 1=true
    Write Parameter Index 2=true
   )
   &ScaleSpriteSize(Scale Factor=&Vector2DFromFloat(Value=&FloatFromCurve(FloatCurve=DataInterface() {propertyValues {Curve {keys [
          {arriveTangent=4.1785, interpMode=RCIM_Cubic, leaveTangent=4.1785, tangentMode=RCTM_User, time=0, value=0}
          {arriveTangent=0.9995, interpMode=RCIM_Cubic, leaveTangent=0.9995, tangentMode=RCTM_User, time=0.2223, value=0.6825}
          {arriveTangent=-0.0405, interpMode=RCIM_Cubic, leaveTangent=-0.0405, tangentMode=RCTM_User, time=1, value=1}
         ]}}})))
   &SpriteRotationRate(Rotation Rate=&Multiply_Float(
     A=&RandomRangeFloat(Minimum=-1, Maximum=1)
     B=&FloatFromCurve(
      FloatCurve=DataInterface() {propertyValues {Curve {keys [
          {arriveTangent=-1.9214, interpMode=RCIM_Cubic, leaveTangent=-1.9214, tangentMode=RCTM_User, time=-0.0046, value=1.0113}
          {arriveTangent=-0.2496, interpMode=RCIM_Cubic, leaveTangent=-0.2496, tangentMode=RCTM_User, time=0.9896, value=0.16}
         ]}}}
      CurveIndex=$Emitter.Age
      Scale Curve=1111
     )
    ))
  ]
  rendererClasses [&NiagaraSpriteRendererProperties]
  renderers [{rendererClass=&NiagaraSpriteRendererProperties}]
 }
 End_Glint {
  emitterUpdateScript [
   &EmitterState()
   &SpawnBurst_Instantaneous(Use Spawn Probability=false, UseLoopCountLimit=false)
  ]
  particleSpawnScript [&InitializeParticle(
    Color=<rgba 0 0.5 3 1>
    Lifetime=0.5
    Position Offset=$System.ArrowTipPosition
    Sprite Size Max=<200 140>
    Sprite Size Min=<120 80>
    Sprite Size Mode=::NewEnumerator2
    UsePositionOffset=true
    Write Material Random=true
   )]
  particleUpdateScript [
   &ParticleState()
   &ScaleSpriteSize(Scale Factor=&Vector2DFromFloat(Value=&FloatFromCurve(
      FloatCurve=DataInterface() {propertyValues {Curve {keys [
          <at 0 0 RCIM_Cubic>
          {arriveTangent=1.4286, interpMode=RCIM_Cubic, leaveTangent=1.4286, time=0.5754, value=0.3493}
          {arriveTangent=-0.8228, interpMode=RCIM_Cubic, leaveTangent=-0.8228, time=0.7, value=1}
          <at 1 0>
         ]}}}
      Scale Curve=0.8
     )))
   &ScaleColor(Scale RGB=&VectorFromFloat(Value=$System.TimeOfDayBrightnessMultiplier), ScaleA=true, ScaleRGB=true)
  ]
  rendererClasses [&NiagaraSpriteRendererProperties]
  renderers [{rendererClass=&NiagaraSpriteRendererProperties}]
 }
 Fireworks_Mesh {
  emitterUpdateScript [
   &EmitterState()
   &SpawnBurst_Instantaneous(Spawn Count=3, Spawn Time=0.1, Use Spawn Probability=false, UseLoopCountLimit=false)
  ]
  particleSpawnScript [
   &InitializeParticle(
    Color=<rgba 0 0.6 2 1>
    Lifetime Max=1
    Lifetime Min=0.6
    Lifetime Mode=::NewEnumerator1
    Mesh Scale Mode=::NewEnumerator1
    Mesh Uniform Scale Max=0.03
    Mesh Uniform Scale Min=0.02
    Position Offset=$System.ArrowTipPosition
    Position Offset Coordinate Space=::Local
    UsePositionOffset=true
    Write Material Random=true
   )
   &InitialMeshOrientation(Use Rotation Vector=false)
  ]
  particleUpdateScript [
   &ParticleState()
   &ScaleColor(
    Scale Alpha=&FloatFromCurve(
     FloatCurve=DataInterface() {propertyValues {Curve {keys [
         {arriveTangent=9.0725, time=0, value=0}
         <at 0.0328 0.8909>
         {arriveTangent=-2.5249, leaveTangent=-2.5249, time=0.1408, value=0.9847}
         {arriveTangent=-1.146, leaveTangent=-1.146, time=0.2, value=0.35}
         <at 0.3784 0.1146>
         {arriveTangent=-0.0405, interpMode=RCIM_Cubic, leaveTangent=-0.0405, tangentMode=RCTM_User, time=1, value=0}
        ]}}}
     Scale Curve=$System.TimeOfDayBrightnessMultiplier
    )
    ScaleA=true
    ScaleRGB=true
   )
   &ScaleMeshSize(Scale Factor=&VectorFromFloat(Value=&FloatFromCurve(FloatCurve=DataInterface() {propertyValues {Curve {keys [{tangentMode=RCTM_User, time=0, value=0}, <at 0.2391 0.6604>, {tangentMode=RCTM_User, time=1, value=1}]}}})))
   &GravityForce(Gravity=<0 0 -33>)
   &SolveForcesAndVelocity(Clamp Velocity=false, Limit Acceleration=false)
   &DynamicMaterialParameters(
    Index 0 Param 1=1
    Index 0 Param 2=1
    Index 0 Param 3=&FloatFromCurve(
     FloatCurve=DataInterface() {propertyValues {Curve {keys [
         {arriveTangent=3.0114, interpMode=RCIM_Cubic, leaveTangent=3.0114, tangentMode=RCTM_User, time=0, value=0}
         {arriveTangent=0.9456, interpMode=RCIM_Cubic, leaveTangent=0.9456, tangentMode=RCTM_User, time=0.4343, value=0.7445}
         {arriveTangent=0.3236, interpMode=RCIM_Cubic, leaveTangent=0.7406, tangentMode=RCTM_User, time=1, value=1}
        ]}}}
     Scale Curve=0.5
    )
    Index 0 Param 4=&FloatFromCurve(FloatCurve=DataInterface() {propertyValues {Curve {keys [
         {arriveTangent=3.4321, interpMode=RCIM_Cubic, leaveTangent=3.4321, tangentMode=RCTM_User, time=0, value=1}
         {arriveTangent=-1.7957, interpMode=RCIM_Cubic, leaveTangent=-1.7957, tangentMode=RCTM_User, time=0.1672, value=1.5}
         {arriveTangent=-0.286, interpMode=RCIM_Cubic, leaveTangent=-0.286, tangentMode=RCTM_User, time=1, value=0.5}
        ]}}})
    Module.Index0Param0Write=true
    Module.Index0Param1Write=true
    Module.Index0Param2Write=true
    Module.Index0Param3Write=true
   )
  ]
  rendererClasses [&NiagaraMeshRendererProperties]
  renderers [{rendererClass=&NiagaraMeshRendererProperties, rendererIndex=0}]
 }
 Flare_With_Streak {
  emitterUpdateScript [
   &EmitterState(Life Cycle Mode=::NewEnumerator1, Loop Behavior=::NewEnumerator1, Loop Duration=1, UseLoopDelay=false)
   &SpawnBurst_Instantaneous(Use Spawn Probability=false, UseLoopCountLimit=false)
  ]
  particleSpawnScript [
   &InitializeParticle(
    Color=<rgba 0 1.5 3 1>
    Lifetime=0.4
    Position Offset=$System.ArrowTipPosition
    Position Offset Coordinate Space=::Local
    Sprite Rotation Mode=::NewEnumerator2
    Sprite Size Mode=::NewEnumerator3
    Uniform Sprite Size=500
    UsePositionOffset=true
    Write Material Random=true
   )
   &DynamicMaterialParameters(
    Index 0 Param 1=&RandomRangeFloat(Minimum=0.4, Maximum=0.6)
    Index 0 Param 2=1
    Index 0 Param 3=1
    Index 0 Param 4=1
    Module.Param0WriteEnabled=true
    Module.Param1WriteEnabled=true
    Module.Param2WriteEnabled=true
    Module.Param3WriteEnabled=true
   )
   /SetVariables() {isSetParameters=true}
  ]
  particleUpdateScript [
   &ParticleState()
   &SolveForcesAndVelocity(Clamp Velocity=false, Limit Acceleration=false)
   &ScaleColor(
    Scale Alpha=&Multiply_Float(
     A=1
     B=&FloatFromCurve(
      FloatCurve=DataInterface() {propertyValues {Curve {keys [<at 0 0>,<at 0.2 1>,<at 0.8 1>,<at 1 0>]}}}
      Scale Curve=$System.TimeOfDayBrightnessMultiplier
     )
    )
    ScaleA=true
    ScaleRGB=true
   )
   &ScaleSpriteSize(Scale Factor=&Vector2DFromFloat(Value=&FloatFromCurve(
      FloatCurve=DataInterface() {propertyValues {Curve {keys [<at 0 1.5 RCIM_Cubic>,<at 1 0>]}}}
      Scale Curve=&FN_FrontEndFloatModifier(Gameplay=1, Front End=0.7)
     )))
   &CameraOffset(Camera Offset Amount=50)
   /SetVariables(Particles.Flash size:Vector2f=&Multiply_Vector2D_ByFloat(
     Float=&Multiply_Float(
      A=400
      B=&FloatFromCurve(
       FloatCurve=DataInterface() {propertyValues {Curve {keys [
           {arriveTangent=4.1785, interpMode=RCIM_Cubic, leaveTangent=4.1785, tangentMode=RCTM_User, time=0, value=0}
           {arriveTangent=0.9995, interpMode=RCIM_Cubic, leaveTangent=0.9995, tangentMode=RCTM_User, time=0.1378, value=0.6825}
           {arriveTangent=-0.0405, interpMode=RCIM_Cubic, leaveTangent=-0.0405, tangentMode=RCTM_User, time=0.62, value=1}
          ]}}}
       Scale Curve=&FN_FrontEndFloatModifier(Gameplay=0.9, Front End=0.7)
      )
     )
     Vector2D=<1.5 1>
    )) {isSetParameters=true}
   /SetVariables(Particles.Flash color:LinearColor=&MakeLinearColorFromVectorAndFloat(
     "Vector (RGB)"=&MakeVectorFromLinearColorRGB(Color=<rgba 0 0.5 1 1>)
     "Float (Alpha)"=&Multiply_Float(
      A=&Multiply_Float(
       A=1
       B=&FloatFromCurve(
        FloatCurve=DataInterface() {propertyValues {Curve {keys [<at 0 0>,<at 0.124 1>,<at 0.496 1>,<at 0.62 0>]}}}
        Scale Curve=$System.TimeOfDayBrightnessMultiplier
       )
      )
      B=1
     )
    )) {isSetParameters=true}
  ]
  rendererClasses [&NiagaraSpriteRendererProperties]
  renderers [
   {rendererClass=&NiagaraSpriteRendererProperties}
   {rendererClass=&NiagaraSpriteRendererProperties, rendererIndex=1}
  ]
 }
 FlashingSparkles {
  emitterUpdateScript [
   &EmitterState()
   &SpawnRate(
    Spawn Probability=&FloatFromCurve(
     FloatCurve=DataInterface() {propertyValues {Curve {keys [
         {interpMode=RCIM_Cubic, leaveTangent=19.2647, tangentMode=RCTM_Break, time=0, value=0}
         <at 0.1 1>
         {arriveTangent=-0.0405, interpMode=RCIM_Cubic, leaveTangent=-0.0405, tangentMode=RCTM_User, time=0.7009, value=1}
         {arriveTangent=-8.0089, interpMode=RCIM_Cubic, leaveTangent=-8.0089, tangentMode=RCTM_User, time=1, value=0}
        ]}}}
     CurveIndex=$System.NormalizedLoopAge
    )
    SpawnRate=33
    Use Spawn Probability=true
   )
  ]
  particleSpawnScript [
   &InitializeParticle(
    AdjustAlpha=false
    AdjustHue=true
    AdjustSaturation=false
    AdjustValue=false
    Color=<rgba 0 0.0833 1 1>
    Color Mode=::NewEnumerator3
    Lifetime Max=0.15
    Lifetime Min=0.2
    Lifetime Mode=::NewEnumerator1
    Saturation Range:Vector2f=<1 1>
    Sprite Rotation Angle=&UniformAOrBFloat(A=0, B=0.125, Evaluation Type=::NewEnumerator0, Recalculate Random Each Loop=true, Override Seed=false)
    Sprite Rotation Mode=::NewEnumerator1
    Sprite Size Mode=::NewEnumerator1
    Uniform Sprite Size Max=66
    Uniform Sprite Size Min=33
    UsePositionOffset=false
    Write Material Random=true
   )
   &ShapeLocation(
    MaskSpawnGroup=false
    Non Uniform Scale=<0.666 0.666 1>
    Offset=<0 0 111>
    Offset Mode=::NewEnumerator0
    Shape Origin=<0 0 0>
    Sphere Distribution=::NewEnumerator1
    Sphere Radius=77
    Sphere Radius Position:NiagaraFloat=&Lerp_Float(A=0, B=1, Alpha=&Power(a=&RandomRangeFloat(Minimum=0, Maximum=1), b=1))
    U Position=&RandomRangeFloat(Minimum=0, Maximum=1)
    UseOwnerScale=true
    V Position=&RandomRangeFloat(Minimum=0, Maximum=1)
   )
   &CameraOffset(Camera Offset Amount=-99)
  ]
  particleUpdateScript [
   &ParticleState()
   &ScaleColor(
    Scale Alpha=&FloatFromCurve(
     FloatCurve=DataInterface() {propertyValues {Curve {keys [
         {arriveTangent=9.8157, interpMode=RCIM_Cubic, leaveTangent=9.8157, tangentMode=RCTM_User, time=0, value=0}
         {arriveTangent=-1.4617, interpMode=RCIM_Cubic, leaveTangent=-1.4617, tangentMode=RCTM_User, time=0.2, value=1}
         {interpMode=RCIM_Cubic, tangentMode=RCTM_User, time=0.3, value=0}
         {interpMode=RCIM_Cubic, tangentMode=RCTM_User, time=0.5, value=0.75}
         {interpMode=RCIM_Cubic, tangentMode=RCTM_Break, time=1, value=0}
        ]}}}
     Scale Curve=$System.TimeOfDayBrightnessMultiplier
    )
    Scale RGB=&VectorFromFloat(Value=$System.TimeOfDayBrightnessMultiplier)
    ScaleA=true
    ScaleRGB=true
   )
   &ScaleSpriteSize(Uniform Curve Sprite Scale=DataInterface() {propertyValues {Curve {keys [{tangentMode=RCTM_User, time=0, value=1}, {tangentMode=RCTM_User, time=1, value=0}]}}})
  ]
  rendererClasses [&NiagaraSpriteRendererProperties]
  renderers [{rendererClass=&NiagaraSpriteRendererProperties}]
 }
 Glints {
  emitterUpdateScript [
   &EmitterState()
   &SpawnBurst_Instantaneous(Spawn Count=12, Use Spawn Probability=false, UseLoopCountLimit=false)
  ]
  particleSpawnScript [
   &InitializeParticle(
    AdjustAlpha=false
    AdjustHue=true
    AdjustSaturation=false
    AdjustValue=false
    Color=<rgba 0 0.0833 1 1>
    Color Mode=::NewEnumerator3
    Lifetime Max=1.1
    Lifetime Min=0.12
    Lifetime Mode=::NewEnumerator1
    Position Offset=$System.ArrowTipPosition
    Position Offset Coordinate Space=::Local
    Sprite Size Max=<200 140>
    Sprite Size Min=<120 80>
    Sprite Size Mode=::NewEnumerator2
    UsePositionOffset=true
    Write Material Random=true
   )
   &ShapeLocation(MaskSpawnGroup=false, Offset Mode=::NewEnumerator0, Override Seed=false, Sphere Radius=10, UseOwnerScale=true)
   &AddVelocity(
    Module.Use Falloff Distance=false
    Origin Offset=$System.ArrowTipPosition
    Override Seed=false
    Velocity Mode=::NewEnumerator1
    Velocity Speed=&RandomRangeFloat(Minimum=200, Maximum=1600)
   )
  ]
  particleUpdateScript [
   &ParticleState()
   &ScaleSpriteSize(Scale Factor=&Vector2DFromFloat(Value=&FloatFromCurve(
      FloatCurve=DataInterface() {propertyValues {Curve {keys [
          <at 0 0 RCIM_Cubic>
          {arriveTangent=1.4286, interpMode=RCIM_Cubic, leaveTangent=1.4286, time=0.5754, value=0.3493}
          {arriveTangent=-0.8228, interpMode=RCIM_Cubic, leaveTangent=-0.8228, time=0.7, value=1}
          <at 1 0>
         ]}}}
      Scale Curve=0.5
     )))
   &Drag(Drag=5, Use Linear Drag=true, Use Rotational Drag=false)
   &SolveForcesAndVelocity(Clamp Velocity=false, Limit Acceleration=false)
   &ScaleColor(
    Scale Alpha=$System.TimeOfDayBrightnessMultiplier
    Scale RGB=&VectorFromFloat(Value=&Multiply_Float(A=2, B=$System.TimeOfDayBrightnessMultiplier))
    ScaleA=true
    ScaleRGB=true
   )
  ]
  rendererClasses [&NiagaraSpriteRendererProperties]
  renderers [{rendererClass=&NiagaraSpriteRendererProperties}]
 }
 Halo {
  emitterUpdateScript [
   &EmitterState()
   &SpawnBurst_Instantaneous(Spawn Time=0.0333, Use Spawn Probability=false, UseLoopCountLimit=false)
  ]
  particleSpawnScript [&InitializeParticle(
    Color=<rgba 0.0521 0.7314 1 1>
    Lifetime=0.45
    Position Offset=$System.ArrowTipPosition
    Position Offset Coordinate Space=::Local
    Sprite Size Mode=::NewEnumerator3
    Uniform Sprite Size=333
    UsePositionOffset=true
    Write Material Random=true
   )]
  particleUpdateScript [
   &ParticleState()
   &ScaleColor(
    Scale Alpha=&Multiply_Float(
     A=&FloatFromCurve(FloatCurve=DataInterface() {propertyValues {Curve {keys [{interpMode=RCIM_Cubic, leaveTangent=-0.6434, tangentMode=RCTM_Break, time=0, value=1}, <at 1 0>]}}})
     B=1
    )
    Scale RGB=&VectorFromFloat(Value=$System.TimeOfDayBrightnessMultiplier)
    ScaleA=true
    ScaleRGB=true
   )
   &CameraOffset(Camera Offset Amount=22)
   &ScaleSpriteSize(Uniform Curve Sprite Scale=DataInterface() {propertyValues {Curve {keys [
        {arriveTangent=4.1785, time=0, value=0}
        <at 0.488 0.7392>
        {arriveTangent=0.0224, interpMode=RCIM_Cubic, leaveTangent=0.0224, tangentMode=RCTM_User, time=1, value=1}
       ]}}})
  ]
  rendererClasses [&NiagaraSpriteRendererProperties]
  renderers [{rendererClass=&NiagaraSpriteRendererProperties}]
 }
 LightRays {
  emitterUpdateScript [
   &EmitterState(Life Cycle Mode=::NewEnumerator1, Loop Behavior=::NewEnumerator1, Loop Delay=0.08, Loop Duration=1, UseLoopDelay=true)
   &SpawnBurst_Instantaneous(Spawn Count=12, Use Spawn Probability=false, UseLoopCountLimit=false)
  ]
  particleSpawnScript [
   &InitializeParticle(
    Color=<rgba 0 0.5 1 1>
    Lifetime Max=0.3
    Lifetime Min=0.2
    Lifetime Mode=::NewEnumerator1
    Position Offset=$System.ArrowTipPosition
    Position Offset Coordinate Space=::Local
    Sprite Rotation Mode=::NewEnumerator2
    Sprite Size Max=<100 222>
    Sprite Size Min=<60 166>
    Sprite Size Mode=::NewEnumerator2
    UsePositionOffset=true
    Write Material Random=true
   )
   &SubUVAnimation(
    End Frame Range Override=3
    Random Start Frame=true
    "Sprite Renderer "=DataInterface() {propertyValues {SpriteRenderer=&self:LightRays_9.NiagaraSpriteRendererProperties_0}}
    UseEndFrame=false
    UseStartFrame=false
   )
  ]
  particleUpdateScript [
   &ParticleState()
   &ScaleSpriteSize(
    Non-Uniform Curve Sprite Scale=DataInterface() {propertyValues {
      XCurve {keys [{tangentMode=RCTM_User, time=0, value=1}, {tangentMode=RCTM_User, time=1, value=0.5}]}
      YCurve {keys [{interpMode=RCIM_Cubic, leaveTangent=0.9587, tangentMode=RCTM_Break, time=0, value=0.5}, {tangentMode=RCTM_User, time=1, value=1}]}
     }}
    Scale Sprite Size Mode=::NewEnumerator3
   )
   &ScaleColor(
    Scale Alpha=&FloatFromCurve(FloatCurve=DataInterface() {propertyValues {Curve {keys [{tangentMode=RCTM_User, time=0, value=1}, {tangentMode=RCTM_User, time=1, value=0}]}}})
    Scale RGB=&VectorFromFloat(Value=&FloatFromCurve(
      FloatCurve=DataInterface() {propertyValues {Curve {keys [<at 0 0>, <at 0.1 5>, {arriveTangent=-92.867, time=0.2, value=1}, <at 1 1>]}}}
      Scale Curve=$System.TimeOfDayBrightnessMultiplier
     ))
    ScaleA=true
    ScaleRGB=true
   )
  ]
  rendererClasses [&NiagaraSpriteRendererProperties]
  renderers [{rendererClass=&NiagaraSpriteRendererProperties}]
 }
 Soft_Shockwave {
  emitterUpdateScript [
   &EmitterState()
   &SpawnBurst_Instantaneous(Use Spawn Probability=false, UseLoopCountLimit=false)
  ]
  particleSpawnScript [&InitializeParticle(
    Color=<rgba 0.5 0 6 3>
    Lifetime=0.333
    Mesh Scale Mode=::NewEnumerator3
    Mesh Uniform Scale=1.5
    Position Offset=$System.ArrowTipPosition
    Position Offset Coordinate Space=::Local
    UsePositionOffset=true
    Write Material Random=true
   )]
  particleUpdateScript [
   &ParticleState()
   &ScaleMeshSize(Scale Factor=&VectorFromFloat(Value=&FloatFromCurve(FloatCurve=DataInterface() {propertyValues {Curve {keys [
          {arriveTangent=4.1785, interpMode=RCIM_Cubic, leaveTangent=4.1785, tangentMode=RCTM_User, time=0, value=0.05}
          {arriveTangent=0.9995, interpMode=RCIM_Cubic, leaveTangent=0.9995, tangentMode=RCTM_User, time=0.3, value=0.6}
          {arriveTangent=-0.0405, interpMode=RCIM_Cubic, leaveTangent=-0.0405, tangentMode=RCTM_User, time=1, value=1}
         ]}}})))
   &ScaleColor(
    Scale Alpha=&FloatFromCurve(
     FloatCurve=DataInterface() {propertyValues {Curve {keys [<at 0 1>, {interpMode=RCIM_Cubic, tangentMode=RCTM_User, time=0.1, value=1}, {interpMode=RCIM_Cubic, tangentMode=RCTM_User, time=1, value=0}]}}}
     Scale Curve=$System.TimeOfDayBrightnessMultiplier
    )
    Scale RGB=&VectorFromFloat(Value=$System.TimeOfDayBrightnessMultiplier)
    ScaleA=true
    ScaleRGB=true
   )
   &DynamicMaterialParameters(
    Index 0 Param 1=1
    Index 0 Param 2=-0.5
    Index 0 Param 3=1
    Index 0 Param 4=1
    Module.Param0WriteEnabled=true
    Module.Param1WriteEnabled=true
    Module.Param2WriteEnabled=true
    Module.Param3WriteEnabled=true
   )
  ]
  rendererClasses [&NiagaraMeshRendererProperties]
  renderers [{rendererClass=&NiagaraMeshRendererProperties, rendererIndex=0}]
 }
 Spark {
  simTarget=GPUComputeSim
  emitterUpdateScript [
   &EmitterState(Life Cycle Mode=::NewEnumerator1, Loop Behavior=::NewEnumerator1, Loop Delay=0.1, Loop Duration=1, UseLoopDelay=true)
   &SpawnBurst_Instantaneous(Spawn Count=333, Use Spawn Probability=false, UseLoopCountLimit=false)
  ]
  particleSpawnScript [
   &InitializeParticle(
    Color=<rgba 0 9.9 33 1>
    Lifetime Max=1.33
    Lifetime Min=0.233
    Lifetime Mode=::NewEnumerator1
    Position Offset=$System.ArrowTipPosition
    Position Offset Coordinate Space=::Local
    Sprite Size Mode=::NewEnumerator1
    Uniform Sprite Size Max=3
    Uniform Sprite Size Min=1
    UsePositionOffset=true
    Write Material Random=true
   )
   &ShapeLocation(MaskSpawnGroup=false, Override Seed=false, Sphere Radius=22, UseOwnerScale=true)
   &AddVelocityFromPoint(
    Module.Use Falloff Distance=false
    Origin Offset=$System.ArrowTipPosition
    Velocity Strength=&RandomRangeFloat(Minimum=111, Maximum=999)
   )
   &CurlNoiseForce(Noise Frequency=99, Noise Strength=666, Pan Noise=true)
   &ApplyInitialForces()
  ]
  particleUpdateScript [
   &ParticleState()
   &GravityForce(Gravity=&RandomRangeVector(Minimum=<0 0 0>, Maximum=<0 0 -100>))
   &Drag(Drag=&RandomRangeFloat(Minimum=6, Maximum=11), Use Linear Drag=true, Use Rotational Drag=true)
   &SolveForcesAndVelocity(Clamp Velocity=false, Limit Acceleration=false)
   &ScaleSpriteSize(Uniform Curve Sprite Scale=DataInterface() {propertyValues {Curve {keys [{tangentMode=RCTM_User, time=0, value=1}, {tangentMode=RCTM_User, time=1, value=0}]}}})
   &ScaleSpriteSizeBySpeed(Max Scale Factor=<0.2 5>, Min Scale Factor=<1 0.5>)
   &ScaleColor(
    Scale Alpha=&FloatFromCurve(FloatCurve=DataInterface() {propertyValues {Curve {keys [<at 0 1>,<at 0.3 0.25>,<at 1 0>]}}})
    Scale RGB=&VectorFromFloat(Value=$System.TimeOfDayBrightnessMultiplier)
    ScaleA=true
    ScaleRGB=true
   )
  ]
  rendererClasses [&NiagaraSpriteRendererProperties]
  renderers [{rendererClass=&NiagaraSpriteRendererProperties}]
 }
 SparkleRainParent {
  emitterUpdateScript [
   &EmitterState()
   &SpawnBurst_Instantaneous(Spawn Count=33, Use Spawn Probability=false, UseLoopCountLimit=false)
  ]
  particleSpawnScript [
   &InitializeParticle(
    AdjustAlpha=false
    AdjustHue=true
    AdjustSaturation=false
    AdjustValue=true
    Color=<rgba 0 0.0833 1 0.666>
    Color Mode=::NewEnumerator3
    Hue Shift Range:Vector2f=<-0.05 0.05>
    Lifetime Max=14
    Lifetime Min=7
    Lifetime Mode=::NewEnumerator1
    Sprite Rotation Angle=0.125
    Sprite Rotation Mode=::NewEnumerator1
    Sprite Size Mode=::NewEnumerator1
    Uniform Sprite Size Max=22
    Uniform Sprite Size Min=11
    UsePositionOffset=false
    Value Range:Vector2f=<0.3 11>
    Write Material Random=true
   )
   &ShapeLocation(
    MaskSpawnGroup=false
    Override Seed=false
    Rotation Mode=::NewEnumerator5
    Scale Mode:ENiagara_ScaleMode=::NewEnumerator1
    Shape Origin=&TransformPosition(Position=Linked() {linkedVariable {name=System.ArrowTipPosition, type=Vector3f}}, Source Space=::Local, Destination Space=::Simulation)
    Sphere Radius=10
   )
   &AddVelocity(
    Velocity=&Multiply_Vector(A=$Particles.ShapeLocation.ShapeVector, B=<1 1 0.333>)
    Velocity Speed Scale=&RandomRangeFloat(Minimum=22, Maximum=66)
   )
   /SetVariables(Particles.TimeOffset:NiagaraFloat=&RandomRangeFloat(Minimum=0, Maximum=1)) {isSetParameters=true}
  ]
  particleUpdateScript [
   &ParticleState()
   /SetVariables(Particles.SparkleTurnTime:NiagaraFloat=&Multiply_Float(A=$Engine.Time, B=1)) {isSetParameters=true}
   /SetVariables(
    Particles.OscillatingColor:LinearColor=&ColorFromCurve(
     ColorCurve=DataInterface() {propertyValues {
       AlphaCurve {keys [<at 0 2 RCIM_Cubic>]}
       BlueCurve {keys [<at 0 1>,<at 0.25 4>,<at 0.5 3>,<at 0.75 4>,<at 1 1 RCIM_Cubic>]}
       GreenCurve {keys [<at 0 0.3>,<at 0.25 0>,<at 0.5 0>,<at 0.75 0>,<at 1 0.3 RCIM_Cubic>]}
       RedCurve {keys [<at 0 0>,<at 0.25 0.2667>,<at 0.5 1.3>,<at 0.75 0.2667>,<at 1 0 RCIM_Cubic>]}
      }}
     CurveIndex=&Frac_Float(Float=&Add_Float(A=$Particles.NormalizedAge, B=&RandomRangeFloat(Minimum=0, Maximum=1)))
    )
    Particles.SpriteRotation:NiagaraFloat=&FloatFromCurve(
     FloatCurve=DataInterface() {propertyValues {Curve {keys [<at 0 0 RCIM_Constant>,<at 0.5 1 RCIM_Constant>,<at 1 0 RCIM_Constant>]}}}
     CurveIndex=&Frac_Float(Float=&Add_Float(A=$Particles.SparkleTurnTime, B=$Particles.TimeOffset))
     Scale Curve=45
    )
    Particles.SpriteRotationOpacity:NiagaraFloat=&FloatFromCurve(
     FloatCurve=DataInterface() {propertyValues {Curve {keys [
         {interpMode=RCIM_Cubic, tangentMode=RCTM_User, time=0, value=0}
         {interpMode=RCIM_Cubic, tangentMode=RCTM_User, time=0.5, value=1}
         {tangentMode=RCTM_User, time=1, value=0}
        ]}}}
     CurveIndex=&Frac_Float(Float=&Multiply_Float(A=&Add_Float(A=$Particles.SparkleTurnTime, B=$Particles.TimeOffset), B=2))
    )
   ) {isSetParameters=true, module=SetVariables_#4}
   &DynamicMaterialParameters(
    Index 0 Param 1=0.8
    Index 0 Param 2=1
    Index 0 Param 3=1
    Index 1 Param 1=150
    Index 1 Param 4=1
    Module.Param0WriteEnabled=true
    Module.Param0WriteEnabled1=true
    Module.Param1WriteEnabled=true
    Module.Param1WriteEnabled1=true
    Module.Param2WriteEnabled=true
    Module.Param2WriteEnabled1=true
    Module.Param3WriteEnabled=false
    Module.Param3WriteEnabled1=true
    Write Parameter Index 1=true
   )
   &Drag(Drag=&RandomRangeFloat(Minimum=2, Maximum=5), Use Linear Drag=true, Use Rotational Drag=true)
   &GravityForce(Gravity=&RandomRangeVector(Minimum=<0 0 -44>, Maximum=<1 1 -99>))
   &CurlNoiseForce(Noise Frequency=10, Noise Strength=100, Pan Noise=false)
   &ScaleColor(
    Color Value To Scale=$Particles.OscillatingColor
    Scale Alpha=&FloatFromCurve(
     FloatCurve=DataInterface() {propertyValues {Curve {keys [
         {tangentMode=RCTM_User, time=0, value=1}
         {arriveTangent=-0.0405, interpMode=RCIM_Cubic, leaveTangent=-0.0405, tangentMode=RCTM_User, time=0.7009, value=1}
         {arriveTangent=-8.0089, interpMode=RCIM_Cubic, leaveTangent=-8.0089, tangentMode=RCTM_User, time=1, value=0}
        ]}}}
     Scale Curve=&Multiply_Float(
      A=$Particles.SpriteRotationOpacity
      B=&Waveform(
       Global Phase=&Add_Float(A=$Engine.Time, B=&RandomRangeFloat(Minimum=0, Maximum=1))
       "Waveform [1]"=::NewEnumerator0
       "[1] Phase"=$Engine.Time
       "[1] Frequency"=&RandomRangeFloat(Minimum=1.618, Maximum=3.618)
       "[1] Amplitude Min/Max"=<0.333 1>
      )
     )
    )
    Scale RGB=&VectorFromFloat(Value=&Multiply_Float(
      A=&Multiply_Float(A=$System.TimeOfDayBrightnessMultiplier, B=$System.Timeline.Time)
      B=&Multiply_Float(
       A=$Particles.SpriteRotationOpacity
       B=&Lerp_Float(
        A=3
        B=3
        Alpha=&FloatFromCurve(FloatCurve=DataInterface() {propertyValues {Curve {keys [<at 0 1>,<at 0.25 0>]}}})
       )
      )
     ))
    ScaleA=true
    ScaleRGB=true
   )
   &ScaleSpriteSize(Uniform Curve Sprite Scale=DataInterface() {propertyValues {Curve {keys [
        {arriveTangent=2.9868, time=0, value=0}
        {leaveTangent=0.2875, time=0.1, value=0.833}
        <at 0.9 1>
        {arriveTangent=-0.0405, leaveTangent=-0.0405, time=1, value=0}
       ]}}})
   &SolveForcesAndVelocity(Clamp Velocity=false, Limit Acceleration=false)
   &KillParticles(Kill Particles=&SetBoolByFloatComparison(A=$System.Timeline.Time, B=0, Comparison Type=::NewEnumerator7))
  ]
  rendererClasses [&NiagaraSpriteRendererProperties]
  renderers [
   {rendererClass=&NiagaraSpriteRendererProperties}
   {rendererClass=&NiagaraSpriteRendererProperties, rendererIndex=1}
  ]
 }
 SparkleRainParentGPU {
  simTarget=GPUComputeSim
  emitterUpdateScript [
   &EmitterState()
   &SpawnBurst_Instantaneous(Spawn Count=66, Use Spawn Probability=false, UseLoopCountLimit=false)
  ]
  particleSpawnScript [
   &InitializeParticle(
    AdjustAlpha=false
    AdjustHue=true
    AdjustSaturation=false
    AdjustValue=true
    Color=<rgba 0 0.0833 1 0.666>
    Color Mode=::NewEnumerator3
    Hue Shift Range:Vector2f=<-0.05 0.05>
    Lifetime Max=14
    Lifetime Min=7
    Lifetime Mode=::NewEnumerator1
    Sprite Rotation Angle=0.125
    Sprite Rotation Mode=::NewEnumerator1
    Sprite Size Mode=::NewEnumerator1
    Uniform Sprite Size Max=22
    Uniform Sprite Size Min=11
    UsePositionOffset=false
    Value Range:Vector2f=<0.3 11>
    Write Material Random=true
   )
   &ShapeLocation(
    MaskSpawnGroup=false
    Override Seed=false
    Rotation Mode=::NewEnumerator5
    Scale Mode:ENiagara_ScaleMode=::NewEnumerator1
    Shape Origin=&TransformPosition(Position=Linked() {linkedVariable {name=System.ArrowTipPosition, type=Vector3f}}, Source Space=::Local, Destination Space=::Simulation)
    Sphere Radius=10
   )
   &AddVelocity(
    Velocity=&Multiply_Vector(A=$Particles.ShapeLocation.ShapeVector, B=<1 1 0.333>)
    Velocity Speed Scale=&RandomRangeFloat(Minimum=22, Maximum=66)
   )
   /SetVariables(Particles.TimeOffset:NiagaraFloat=&RandomRangeFloat(Minimum=0, Maximum=1)) {isSetParameters=true}
  ]
  particleUpdateScript [
   &ParticleState()
   /SetVariables(Particles.SparkleTurnTime:NiagaraFloat=&Multiply_Float(A=$Engine.Time, B=2)) {isSetParameters=true}
   /SetVariables(
    Particles.OscillatingColor:LinearColor=&ColorFromCurve(
     ColorCurve=DataInterface() {propertyValues {
       AlphaCurve {keys [<at 0 2 RCIM_Cubic>]}
       BlueCurve {keys [<at 0 1>,<at 0.25 4>,<at 0.5 3>,<at 0.75 4>,<at 1 1 RCIM_Cubic>]}
       GreenCurve {keys [<at 0 0.3>,<at 0.25 0>,<at 0.5 0>,<at 0.75 0>,<at 1 0.3 RCIM_Cubic>]}
       RedCurve {keys [<at 0 0>,<at 0.25 0.2667>,<at 0.5 1.3>,<at 0.75 0.2667>,<at 1 0 RCIM_Cubic>]}
      }}
     CurveIndex=&Frac_Float(Float=&Add_Float(A=$Particles.NormalizedAge, B=&RandomRangeFloat(Minimum=0, Maximum=1)))
    )
    Particles.SpriteRotation:NiagaraFloat=&FloatFromCurve(
     FloatCurve=DataInterface() {propertyValues {Curve {keys [<at 0 0 RCIM_Constant>,<at 0.5 1 RCIM_Constant>,<at 1 0 RCIM_Constant>]}}}
     CurveIndex=&Frac_Float(Float=&Add_Float(A=$Particles.SparkleTurnTime, B=$Particles.TimeOffset))
     Scale Curve=45
    )
    Particles.SpriteRotationOpacity:NiagaraFloat=&FloatFromCurve(
     FloatCurve=DataInterface() {propertyValues {Curve {keys [
         {interpMode=RCIM_Cubic, tangentMode=RCTM_User, time=0, value=0}
         {interpMode=RCIM_Cubic, tangentMode=RCTM_User, time=0.5, value=1}
         {tangentMode=RCTM_User, time=1, value=0}
        ]}}}
     CurveIndex=&Frac_Float(Float=&Multiply_Float(A=&Add_Float(A=$Particles.SparkleTurnTime, B=$Particles.TimeOffset), B=2))
    )
   ) {isSetParameters=true, module=SetVariables_#4}
   &DynamicMaterialParameters(
    Index 0 Param 1=0.8
    Index 0 Param 2=1
    Index 0 Param 3=1
    Index 1 Param 1=150
    Index 1 Param 4=1
    Module.Param0WriteEnabled=true
    Module.Param0WriteEnabled1=true
    Module.Param1WriteEnabled=true
    Module.Param1WriteEnabled1=true
    Module.Param2WriteEnabled=true
    Module.Param2WriteEnabled1=true
    Module.Param3WriteEnabled=false
    Module.Param3WriteEnabled1=true
    Write Parameter Index 1=true
   )
   &Drag(Drag=&RandomRangeFloat(Minimum=2, Maximum=5), Use Linear Drag=true, Use Rotational Drag=true)
   &GravityForce(Gravity=&RandomRangeVector(Minimum=<0 0 -44>, Maximum=<1 1 -99>))
   &CurlNoiseForce(Noise Frequency=10, Noise Strength=100, Pan Noise=false)
   &ScaleColor(
    Color Value To Scale=$Particles.OscillatingColor
    Scale Alpha=&FloatFromCurve(
     FloatCurve=DataInterface() {propertyValues {Curve {keys [
         {tangentMode=RCTM_User, time=0, value=1}
         {arriveTangent=-0.0405, interpMode=RCIM_Cubic, leaveTangent=-0.0405, tangentMode=RCTM_User, time=0.7009, value=1}
         {arriveTangent=-8.0089, interpMode=RCIM_Cubic, leaveTangent=-8.0089, tangentMode=RCTM_User, time=1, value=0}
        ]}}}
     Scale Curve=&Multiply_Float(
      A=$Particles.SpriteRotationOpacity
      B=&Waveform(
       Global Phase=&Add_Float(A=$Engine.Time, B=&RandomRangeFloat(Minimum=0, Maximum=1))
       "Waveform [1]"=::NewEnumerator0
       "[1] Phase"=$Engine.Time
       "[1] Frequency"=&RandomRangeFloat(Minimum=1.618, Maximum=3.618)
       "[1] Amplitude Min/Max"=<0.333 1>
      )
     )
    )
    Scale RGB=&VectorFromFloat(Value=&Multiply_Float(
      A=&Multiply_Float(A=$System.TimeOfDayBrightnessMultiplier, B=$System.Timeline.Time)
      B=&Multiply_Float(
       A=$Particles.SpriteRotationOpacity
       B=&Lerp_Float(
        A=3
        B=3
        Alpha=&FloatFromCurve(FloatCurve=DataInterface() {propertyValues {Curve {keys [<at 0 1>,<at 0.25 0>]}}})
       )
      )
     ))
    ScaleA=true
    ScaleRGB=true
   )
   &ScaleSpriteSize(Uniform Curve Sprite Scale=DataInterface() {propertyValues {Curve {keys [
        {arriveTangent=2.9868, time=0, value=0}
        {leaveTangent=0.2875, time=0.1, value=0.833}
        <at 0.9 1>
        {arriveTangent=-0.0405, leaveTangent=-0.0405, time=1, value=0}
       ]}}})
   &SolveForcesAndVelocity(Clamp Velocity=false, Limit Acceleration=false)
   &KillParticles(Kill Particles=&SetBoolByFloatComparison(A=$System.Timeline.Time, B=0, Comparison Type=::NewEnumerator7))
  ]
  rendererClasses [&NiagaraSpriteRendererProperties]
  renderers [
   {rendererClass=&NiagaraSpriteRendererProperties}
   {rendererClass=&NiagaraSpriteRendererProperties, rendererIndex=1}
  ]
 }
 SparkleRainTrail {
  emitterUpdateScript [
   &EmitterState()
   &SpawnParticlesFromOtherEmitter(
    Attribute Reader=DataInterface() {propertyValues {EmitterBinding {bindingMode=Other, emitterName=SparkleRainParent}}}
    Maximum Age Spawn Threshold=2
    Minimum Age Spawn Threshold=0
    Spawn Rate=22
    Spawn Rate Per Particle Cap=11
    Use Maximum Age=false
    Use Minimum Age=false
   )
  ]
  particleSpawnScript [
   &InitializeParticle(
    Color Mode=::NewEnumerator0
    Lifetime Max=1.333
    Lifetime Min=0.333
    Lifetime Mode=::NewEnumerator1
    Sprite Rotation Angle=0.125
    Sprite Rotation Mode=::NewEnumerator1
    Sprite Size Mode=::NewEnumerator1
    Uniform Sprite Size Max=11
    Uniform Sprite Size Min=11
    UsePositionOffset=false
    Write Material Random=true
   )
   &SampleParticlesFromOtherEmitter(
    Applied Alpha Scale=0.1
    Applied RGB Scale:Vector3f=<3 3 3>
    Applied Velocity Scale:NiagaraFloat=&FloatFromCurve(FloatCurve=DataInterface() {propertyValues {Curve {keys [<at 0 1>,<at 1 0>]}}}, Scale Curve=0.5)
    Color Sampling=::NewEnumerator0
    Velocity Sampling=::NewEnumerator0
   )
   &ShapeLocation(MaskSpawnGroup=false, Override Seed=false, Sphere Radius=5, UseOwnerScale=true)
  ]
  particleUpdateScript [
   &ParticleState()
   &Drag(Drag=2, Use Linear Drag=true, Use Rotational Drag=false)
   &GravityForce(Gravity=<0 0 -22>)
   &CurlNoiseForce(Noise Frequency=1, Noise Strength=15, Pan Noise=false)
   &SolveForcesAndVelocity(Clamp Velocity=false, Limit Acceleration=false)
   &ScaleColor(
    Scale Alpha=&FloatFromCurve(
     FloatCurve=DataInterface() {propertyValues {Curve {keys [
         {tangentMode=RCTM_User, time=0, value=1}
         {arriveTangent=-0.0405, interpMode=RCIM_Cubic, leaveTangent=-0.0405, tangentMode=RCTM_User, time=0.7009, value=1}
         {arriveTangent=-8.0089, interpMode=RCIM_Cubic, leaveTangent=-8.0089, tangentMode=RCTM_User, time=1, value=0}
        ]}}}
     Scale Curve=&Waveform(
      Global Input Position=&RandomRangeFloat(Minimum=0, Maximum=1)
      Global Phase=&Add_Float(A=$Particles.Age, B=&RandomRangeFloat(Minimum=0, Maximum=1))
      "Waveform [1]"=::NewEnumerator0
      "[1] Frequency"=&RandomRangeFloat(Minimum=3.618, Maximum=11.08)
      "[1] Amplitude Min/Max"=<0 1>
     )
    )
    Scale RGB=&VectorFromFloat(Value=&Multiply_Float(
      A=$System.TimeOfDayBrightnessMultiplier
      B=&Lerp_Float(
       A=0.25
       B=22
       Alpha=&FloatFromCurve(FloatCurve=DataInterface() {propertyValues {Curve {keys [<at 0 1>,<at 0.333 0>]}}})
      )
     ))
    ScaleA=true
    ScaleRGB=true
   )
   &ScaleSpriteSize(Uniform Curve Sprite Scale=DataInterface() {propertyValues {Curve {keys [
        {interpMode=RCIM_Cubic, tangentMode=RCTM_User, time=0, value=0}
        {interpMode=RCIM_Cubic, tangentMode=RCTM_User, time=0.5, value=1}
        {tangentMode=RCTM_User, time=1, value=0}
       ]}}})
   &KillParticles(Kill Particles=&SetBoolByFloatComparison(A=$System.Timeline.Time, B=0, Comparison Type=::NewEnumerator7))
  ]
  rendererClasses [&NiagaraSpriteRendererProperties]
  renderers [{rendererClass=&NiagaraSpriteRendererProperties}]
 }
 SparkleRainTrailGPU {
  simTarget=GPUComputeSim
  emitterUpdateScript [
   &EmitterState()
   &SpawnParticlesFromOtherEmitter(
    Attribute Reader=DataInterface() {propertyValues {EmitterBinding {bindingMode=Other, emitterName=SparkleRainParentGPU}}}
    Maximum Age Spawn Threshold=2
    Minimum Age Spawn Threshold=0
    Spawn Rate=99
    Spawn Rate Per Particle Cap=99
    Use Maximum Age=false
    Use Minimum Age=false
   )
  ]
  particleSpawnScript [
   &InitializeParticle(
    Color Mode=::NewEnumerator0
    Lifetime Max=1.333
    Lifetime Min=0.333
    Lifetime Mode=::NewEnumerator1
    Sprite Rotation Angle=0.125
    Sprite Rotation Mode=::NewEnumerator1
    Sprite Size Mode=::NewEnumerator1
    Uniform Sprite Size Max=11
    Uniform Sprite Size Min=1
    UsePositionOffset=false
    Write Material Random=true
   )
   &SampleParticlesFromOtherEmitter(
    Applied Alpha Scale=0.3
    Applied Velocity Scale:NiagaraFloat=&FloatFromCurve(
     FloatCurve=DataInterface() {propertyValues {Curve {keys [<at 0 1>,<at 1 0>]}}}
     Scale Curve=&RandomRangeFloat(Minimum=-0.1, Maximum=0.75)
    )
    Color Sampling=::NewEnumerator0
    Velocity Sampling=::NewEnumerator0
   )
   &ShapeLocation(MaskSpawnGroup=false, Override Seed=false, Sphere Radius=2, UseOwnerScale=true)
  ]
  particleUpdateScript [
   &ParticleState()
   &Drag(Drag=&RandomRangeFloat(Minimum=1, Maximum=2), Use Linear Drag=true, Use Rotational Drag=true)
   &GravityForce(Gravity=<0 0 -22>)
   &CurlNoiseForce(Noise Frequency=1, Noise Strength=&RandomRangeFloat(Minimum=15, Maximum=33), Pan Noise=false)
   &SolveForcesAndVelocity(Clamp Velocity=false, Limit Acceleration=false)
   &ScaleColor(
    Scale Alpha=&FloatFromCurve(
     FloatCurve=DataInterface() {propertyValues {Curve {keys [
         {tangentMode=RCTM_User, time=0, value=1}
         {arriveTangent=-0.0405, interpMode=RCIM_Cubic, leaveTangent=-0.0405, tangentMode=RCTM_User, time=0.7009, value=1}
         {arriveTangent=-8.0089, interpMode=RCIM_Cubic, leaveTangent=-8.0089, tangentMode=RCTM_User, time=1, value=0}
        ]}}}
     Scale Curve=&Waveform(
      Global Input Position=&RandomRangeFloat(Minimum=0, Maximum=1)
      Global Phase=&Add_Float(A=$Particles.Age, B=&RandomRangeFloat(Minimum=0, Maximum=1))
      "Waveform [1]"=::NewEnumerator0
      "[1] Frequency"=&RandomRangeFloat(Minimum=3.618, Maximum=11.08)
      "[1] Amplitude Min/Max"=<0 1>
     )
    )
    Scale RGB=&VectorFromFloat(Value=&Multiply_Float(
      A=$System.TimeOfDayBrightnessMultiplier
      B=&Lerp_Float(
       A=0.25
       B=22
       Alpha=&FloatFromCurve(FloatCurve=DataInterface() {propertyValues {Curve {keys [<at 0 1>,<at 0.333 0>]}}})
      )
     ))
    ScaleA=true
    ScaleRGB=true
   )
   &ScaleSpriteSize(Uniform Curve Sprite Scale=DataInterface() {propertyValues {Curve {keys [
        {interpMode=RCIM_Cubic, tangentMode=RCTM_User, time=0, value=0}
        {interpMode=RCIM_Cubic, tangentMode=RCTM_User, time=0.5, value=1}
        {tangentMode=RCTM_User, time=1, value=0}
       ]}}})
   &KillParticles(Kill Particles=&SetBoolByFloatComparison(A=$System.Timeline.Time, B=0, Comparison Type=::NewEnumerator7))
  ]
  rendererClasses [&NiagaraSpriteRendererProperties]
  renderers [{rendererClass=&NiagaraSpriteRendererProperties}]
 }
}
summary {
 systemName=NS_Emote_SeleneCobra_Loop
 emitters [
  {
   emitterName=ArrowExplosionLight
   rendererClasses [
    &NiagaraSpriteRendererProperties
    &NiagaraLightRendererProperties
    &NiagaraComponentRendererProperties
   ]
  }
  {emitterName=FlashingSparkles, rendererClasses [&NiagaraSpriteRendererProperties]}
  {emitterName=SparkleRainParent, rendererClasses [&NiagaraSpriteRendererProperties]}
  {emitterName=SparkleRainTrail, rendererClasses [&NiagaraSpriteRendererProperties]}
  {emitterName=DustPuffFeet, rendererClasses [&NiagaraSpriteRendererProperties]}
  {emitterName=Fireworks_Mesh, rendererClasses [&NiagaraMeshRendererProperties]}
  {emitterName=Glints, rendererClasses [&NiagaraSpriteRendererProperties]}
  {emitterName=Dust, rendererClasses [&NiagaraSpriteRendererProperties]}
  {emitterName=Halo, rendererClasses [&NiagaraSpriteRendererProperties]}
  {emitterName=Flare_With_Streak, rendererClasses [&NiagaraSpriteRendererProperties]}
  {emitterName=Spark, simTarget=GPUComputeSim, rendererClasses [&NiagaraSpriteRendererProperties]}
  {emitterName=BG_Glow, rendererClasses [&NiagaraSpriteRendererProperties]}
  {emitterName=LightRays, rendererClasses [&NiagaraSpriteRendererProperties]}
  {emitterName=End_Glint, rendererClasses [&NiagaraSpriteRendererProperties]}
  {emitterName=Soft_Shockwave, rendererClasses [&NiagaraMeshRendererProperties]}
  {emitterName=SparkleRainParentGPU, simTarget=GPUComputeSim, rendererClasses [&NiagaraSpriteRendererProperties]}
  {emitterName=SparkleRainTrailGPU, simTarget=GPUComputeSim, rendererClasses [&NiagaraSpriteRendererProperties]}
 ]
 userVariables [
  User.ArrowTipStartingOffset {type=Vector3f, description="", defaultValue {struct=Vector3f, value=<5 13 214>}}
  User.bIsFrontEnd {type=NiagaraBool}
  User.bIsFrontEndPreview {type=NiagaraBool}
 ]
}
